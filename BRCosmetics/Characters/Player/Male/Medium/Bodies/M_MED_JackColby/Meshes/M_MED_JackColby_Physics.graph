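# Physics asset for the M_MED_JackColby player body cosmetic, covering the lower torso and a dynamic back strap.
# Capsules on the pelvis and two spine bones form the body volume, linked by free-rotation constraints.
# A zero-size base body joins the three-segment strap chain to spine_02, with narrow twist and moderate swing limits along the chain.
# Each strap segment pairs a thin box with an offset capsule.
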PhysicsAsset /BRCosmetics/Characters/Player/Male/Medium/Bodies/M_MED_JackColby/Meshes/M_MED_JackColby_Physics
bodies {
 dyn_strap_bk_01 [
  {shapeType=Capsule, length=5, radius=4, shapeName=dyn_strap_bk_01_capsule, center=<5.2845 3.1342 0>, rotation=<rot 90 0 0>}
  {shapeType=Box, extentX=7.5, extentY=2.5, extentZ=7.5, shapeName=dyn_strap_bk_01_box, center=<4.4498 0 0>}
 ]
 dyn_strap_bk_02 [
  {shapeType=Capsule, length=5, radius=4, shapeName=dyn_strap_bk_02_capsule, center=<5.2845 3.2095 0>, rotation=<rot 90 0 0>}
  {shapeType=Box, extentX=7.5, extentY=2.5, extentZ=7.5, shapeName=dyn_strap_bk_02_box, center=<4.4498 0 0>}
 ]
 dyn_strap_bk_03 [
  {shapeType=Capsule, length=5, radius=5, shapeName=dyn_strap_bk_03_capsule, center=<5.4931 3.8059 0>, rotation=<rot 90 0 0>}
  {shapeType=Box, extentX=10.5, extentY=2.5, extentZ=7.5, shapeName=dyn_strap_bk_03_box, center=<4.4498 0 0>}
 ]
 dyn_strap_bk_base [{shapeType=Capsule, length=0, radius=0, shapeName=dyn_strap_bk_base_capsule, center=<0.0463 0 0>}]
 pelvis [{shapeType=Capsule, length=15.3977, radius=12.0191, shapeName=pelvis_capsule, center=<-2.7045 -0.7511 0>}]
 spine_01 [{shapeType=Capsule, length=9.005, radius=11.005, shapeName=spine_01_capsule, center=<-3.4973 -0.6397 0>}]
 spine_02 [{shapeType=Capsule, length=10.2886, radius=10.2977, shapeName=spine_02_capsule}]
}
constraints [
 {
  bone1Name=spine_01
  bone2Name=pelvis
  swing1LimitDegrees=45
  swing1Motion=Free
  swing2LimitDegrees=45
  swing2Motion=Free
  twistLimitDegrees=45
  twistMotion=Free
 }
 {
  bone1Name=spine_02
  bone2Name=spine_01
  swing1LimitDegrees=45
  swing1Motion=Free
  swing2LimitDegrees=45
  swing2Motion=Free
  twistLimitDegrees=45
  twistMotion=Free
 }
 {bone1Name=dyn_strap_bk_base, bone2Name=spine_02, swing1LimitDegrees=45, swing2LimitDegrees=25, twistLimitDegrees=5}
 {bone1Name=dyn_strap_bk_01, bone2Name=dyn_strap_bk_base, swing1LimitDegrees=30, swing2LimitDegrees=30, twistLimitDegrees=5}
 {bone1Name=dyn_strap_bk_02, bone2Name=dyn_strap_bk_01, swing1LimitDegrees=30, swing2LimitDegrees=30, twistLimitDegrees=10}
 {bone1Name=dyn_strap_bk_03, bone2Name=dyn_strap_bk_02, swing1LimitDegrees=25, swing2LimitDegrees=15, twistLimitDegrees=10}
]
