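# Animation Blueprint for the M_MED_JackColby player body that adds secondary motion to the pose copied from its parent mesh.
# A Control Rig driven by left/right movement feeds a RigidBody simulation, disabled when the backbling is a cape.
# AnimDynamics then swings laces, zippers, necklace and pistol bones, before weighted constraints adjust lace, strap and zipper base bones.
# The update event computes movement, the cape flag and state-based gravity through a deprecated helper.
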AnimBlueprint /BRCosmetics/Characters/Player/Male/Medium/Bodies/M_MED_JackColby/Meshes/M_MED_JackColby_AnimBP
target_skeleton=&/M_MED_JackColby_Skeleton
graphs [
 AnimGraph {nodes [
   n1 = "Constraint - dyn_lace_l_base (PPP)"!(
    ComponentPose=Local To Component(LocalPose=Use cached pose 'toAnimDyn01'())
    Alpha=1
    ConstraintWeights_0=0.6104
    ConstraintWeights_1:float=0.323
    ConstraintWeights_2:float=0.0666
   ) -> (Pose) {
    alphaBoolBlend {blendOption=Linear}
    boneToModify {boneName=dyn_lace_l_base}
    constraintSetup [{targetBone {boneName=calf_twist_01_l}}, {targetBone {boneName=calf_twist_02_l}}, {targetBone {boneName=foot_l}}]
    constraintWeights [1,1,1]
   }
   n2 = "Constraint - dyn_lace_r_base (PPP)"!(ComponentPose=@n1, Alpha=1, ConstraintWeights_0=0.6104, ConstraintWeights_1:float=0.323, ConstraintWeights_2:float=0.0666) -> (Pose) {
    alphaBoolBlend {blendOption=Linear}
    boneToModify {boneName=dyn_lace_r_base}
    constraintSetup [{targetBone {boneName=calf_twist_01_r}}, {targetBone {boneName=calf_twist_02_r}}, {targetBone {boneName=foot_r}}]
    constraintWeights [1,1,1]
   }
   n3 = "Constraint - dyn_strap_bk_base (PPPPP)"!(
    ComponentPose=@n2
    Alpha=1
    ConstraintWeights_0=0.3481
    ConstraintWeights_1:float=0.2295
    ConstraintWeights_2:float=0.2482
    ConstraintWeights_3:float=0.1643
    ConstraintWeights_4:float=0.0099
   ) -> (Pose) {
    alphaBoolBlend {blendOption=Linear}
    boneToModify {boneName=dyn_strap_bk_base}
    constraintSetup [
     {targetBone {boneName=pelvis}}
     {targetBone {boneName=spine_01}}
     {targetBone {boneName=spine_02}}
     {targetBone {boneName=spine_03}}
     {targetBone {boneName=spine_04}}
    ]
    constraintWeights [1,1,1,1,1]
   }
   n4 = "Constraint - dyn_suit_frnt_zippo_r_base (PPPPP)"!(
    ComponentPose=@n3
    Alpha=1
    ConstraintWeights_0=0.3873
    ConstraintWeights_1:float=0.1935
    ConstraintWeights_2:float=0.246
    ConstraintWeights_3:float=0.1539
    ConstraintWeights_4:float=0.0193
   ) -> (Pose) {
    alphaBoolBlend {blendOption=Linear}
    boneToModify {boneName=dyn_suit_frnt_zippo_r_base}
    constraintSetup [
     {targetBone {boneName=pelvis}}
     {targetBone {boneName=spine_01}}
     {targetBone {boneName=spine_02}}
     {targetBone {boneName=spine_03}}
     {targetBone {boneName=spine_04}}
    ]
    constraintWeights [1,1,1,1,1]
   }
   n5 = "Constraint - dyn_shoulder_zippo_l_base (PPPPPPPP)"!(
    ComponentPose=@n4
    Alpha=1
    ConstraintWeights_0=0.0227
    ConstraintWeights_1:float=0.018
    ConstraintWeights_2:float=0.1461
    ConstraintWeights_3:float=0.0159
    ConstraintWeights_4:float=0.4041
    ConstraintWeights_5:float=0.2672
    ConstraintWeights_6:float=0.034
    ConstraintWeights_7:float=0.092
   ) -> (Pose) {
    alphaBoolBlend {blendOption=Linear}
    boneToModify {boneName=dyn_shoulder_zippo_l_base}
    constraintSetup [
     {targetBone {boneName=spine_03}}
     {targetBone {boneName=spine_05}}
     {targetBone {boneName=clavicle_l}}
     {targetBone {boneName=upperarm_l}}
     {targetBone {boneName=upperarm_twist_01_l}}
     {targetBone {boneName=deform_upperarm_up_l}}
     {targetBone {boneName=deform_upperarm_bk_l}}
     {targetBone {boneName=upperarm_twist_02_l}}
    ]
    constraintWeights [1,1,1,1,1,1,1,1]
   }
   n6 = "Constraint - dyn_shoulder_zippo_r_base (PPPPPPPP)"!(
    ComponentPose=@n5
    Alpha=1
    ConstraintWeights_0=0.021
    ConstraintWeights_1:float=0.0156
    ConstraintWeights_2:float=0.1381
    ConstraintWeights_3:float=0.0197
    ConstraintWeights_4:float=0.4185
    ConstraintWeights_5:float=0.2581
    ConstraintWeights_6:float=0.0304
    ConstraintWeights_7:float=0.0986
   ) -> (Pose) {
    alphaBoolBlend {blendOption=Linear}
    boneToModify {boneName=dyn_shoulder_zippo_r_base}
    constraintSetup [
     {targetBone {boneName=spine_03}}
     {targetBone {boneName=spine_05}}
     {targetBone {boneName=clavicle_r}}
     {targetBone {boneName=upperarm_r}}
     {targetBone {boneName=upperarm_twist_01_r}}
     {targetBone {boneName=deform_upperarm_up_r}}
     {targetBone {boneName=deform_upperarm_bk_r}}
     {targetBone {boneName=upperarm_twist_02_r}}
    ]
    constraintWeights [1,1,1,1,1,1,1,1]
   }
   Output Pose(Result=Component To Local(ComponentPose=@n6)) {layerGroup=None}
   n7 = Control Rig!(
    Source=Copy Pose From Mesh() {bUseAttachedParent=true}
    Alpha
    LeftRight Movement:double=Get Left Right Movement() -> (Left Right Movement:double)
   ) -> (Pose) {
    alphaBoolBlend {blendOption=Linear}
    controlRigAssetReference {blueprintRigClass=&/M_MED_JackColby_CtrlRig.M_MED_JackColby_CtrlRig_C}
   }
   n8 = RigidBody!(
    ComponentPose=Local To Component(LocalPose=@n7)
    bAlphaBoolEnabled:bool=Get Is Not Cape() -> (Is Not Cape:bool)
    OverrideWorldGravity:Vector="0,0,0"
    ExternalForce
   ) -> (Pose) {
    alphaInputType=Bool
    bClampLinearTranslationLimitToRefPose=true
    bOverrideWorldGravity=true
    overridePhysicsAsset=&/M_MED_JackColby_Physics
    simulationSpace=BaseBoneSpace
    alphaBoolBlend {blendOption=Linear}
    baseBoneRef {boneName=pelvis}
    componentAppliedLinearAccClamp=<3000 3000 3000>
    simSpaceSettings {
     dampingAlpha=0.5
     maxAngularAcceleration=350
     maxAngularVelocity=10
     maxLinearAcceleration=1500
     maxLinearVelocity=1200
     velocityScaleZ=10
     worldAlpha=1
     externalLinearDragV=<1 0.75 0.2>
    }
   }
   Save cached pose 'toCtrlRig'(Pose=Component To Local(ComponentPose=@n8)) {cacheName=toCtrlRig}
   n9 = Anim Dynamics - Bone: dyn_lace_l_01(ComponentPose=Local To Component(LocalPose=Use cached pose 'toCtrlRig'()), Alpha=1, ExternalForce="0,0,0") {
    angularDampingOverride=0.999
    angularSpringConstant=50
    bAngularSpring=true
    bGravityOverrideInSimSpace=true
    bOverrideAngularDamping=true
    bUseGravityOverride=true
    bUsePlanarLimit=false
    simulationSpace=RootRelative
    alphaBoolBlend {blendOption=Linear}
    boundBone {boneName=dyn_lace_l_01}
    componentAppliedLinearAccClamp=<2000 2000 2000>
    gravityOverride {z=-980}
    physicsBodyDefinitions [{
      boundBone {boneName=dyn_lace_l_01}
      boxExtents=<4 4 4>
      constraintSetup {angularTargetAxis=AxisZ, angularLimitsMax=<30 40 30>, angularLimitsMin=<-30 0 -30>, angularTarget=<0 0 1>}
      localJointOffset=<0 0 -1>
     }]
    relativeSpaceBone {boneName=calf_l}
   }
   n10 = Anim Dynamics - Bone: dyn_lace_r_01(ComponentPose=@n9, Alpha=1, ExternalForce="0,0,0") {
    angularDampingOverride=0.999
    angularSpringConstant=50
    bAngularSpring=true
    bGravityOverrideInSimSpace=true
    bOverrideAngularDamping=true
    bUseGravityOverride=true
    bUsePlanarLimit=false
    simulationSpace=RootRelative
    alphaBoolBlend {blendOption=Linear}
    boundBone {boneName=dyn_lace_r_01}
    componentAppliedLinearAccClamp=<2000 2000 2000>
    gravityOverride {z=-980}
    physicsBodyDefinitions [{
      boundBone {boneName=dyn_lace_r_01}
      boxExtents=<4 4 4>
      constraintSetup {angularTargetAxis=AxisZ, angularLimitsMax=<30 40 30>, angularLimitsMin=<-30 0 -30>, angularTarget=<0 0 1>}
      localJointOffset=<0 0 -1>
     }]
   }
   n11 = Anim Dynamics - Bone: dyn_bags_zippo_up_01(ComponentPose=@n10, Alpha=1, ExternalForce="0,0,0") {
    angularDampingOverride=0.999
    angularSpringConstant=200
    bOverrideAngularDamping=true
    bUseGravityOverride=true
    simulationSpace=RootRelative
    alphaBoolBlend {blendOption=Linear}
    boundBone {boneName=dyn_bags_zippo_up_01}
    componentAppliedLinearAccClamp=<2000 2000 2000>
    componentLinearAccScale=<1 1 1>
    componentLinearVelScale=<1 1 1>
    gravityOverride {z=-980}
    physicsBodyDefinitions [{
      boundBone {boneName=dyn_bags_zippo_up_01}
      boxExtents=<2 2 2>
      constraintSetup {angularLimitsMax=<10 40 20>, angularLimitsMin=<-10 0 -20>, angularTarget=<1 0 0>}
      localJointOffset=<0.1 0 0>
     }]
   }
   n12 = Anim Dynamics - Bone: dyn_bags_zippo_down_01(ComponentPose=@n11, Alpha=1, ExternalForce="0,0,0") {
    angularDampingOverride=0.999
    angularSpringConstant=200
    bOverrideAngularDamping=true
    bUseGravityOverride=true
    simulationSpace=RootRelative
    alphaBoolBlend {blendOption=Linear}
    boundBone {boneName=dyn_bags_zippo_down_01}
    componentAppliedLinearAccClamp=<2000 2000 2000>
    componentLinearAccScale=<1 1 1>
    componentLinearVelScale=<1 1 1>
    gravityOverride {z=-980}
    physicsBodyDefinitions [{
      boundBone {boneName=dyn_bags_zippo_down_01}
      boxExtents=<2 2 2>
      constraintSetup {angularLimitsMax=<10 40 20>, angularLimitsMin=<-10 0 -20>, angularTarget=<1 0 0>}
      localJointOffset=<0.1 0 0>
     }]
   }
   n13 = Anim Dynamics - Bone: dyn_suit_frnt_zippo_r_01(ComponentPose=@n12, Alpha=1, ExternalForce="0,0,0") {
    angularDampingOverride=1
    angularSpringConstant=200
    bOverrideAngularDamping=true
    bUseGravityOverride=true
    simulationSpace=RootRelative
    alphaBoolBlend {blendOption=Linear}
    boundBone {boneName=dyn_suit_frnt_zippo_r_01}
    componentAppliedLinearAccClamp=<2000 2000 2000>
    componentLinearAccScale=<1 1 1>
    componentLinearVelScale=<1 1 1>
    gravityOverride {z=-980}
    physicsBodyDefinitions [{
      boundBone {boneName=dyn_suit_frnt_zippo_r_01}
      boxExtents=<2 2 2>
      constraintSetup {angularLimitsMax=<10 10 20>, angularLimitsMin=<-10 -20 -20>, angularTarget=<1 0 0>}
      localJointOffset=<0.1 0 0>
     }]
   }
   n14 = Anim Dynamics - Bone: dyn_shoulder_zippo_l_01(ComponentPose=@n13, Alpha=1, ExternalForce="0,0,0") {
    angularDampingOverride=1
    angularSpringConstant=200
    bOverrideAngularDamping=true
    bUseGravityOverride=true
    simulationSpace=RootRelative
    alphaBoolBlend {blendOption=Linear}
    boundBone {boneName=dyn_shoulder_zippo_l_01}
    componentAppliedLinearAccClamp=<2000 2000 2000>
    componentLinearAccScale=<1 1 1>
    componentLinearVelScale=<1 1 1>
    gravityOverride {z=-980}
    physicsBodyDefinitions [{
      boundBone {boneName=dyn_shoulder_zippo_l_01}
      boxExtents=<2 2 2>
      constraintSetup {angularLimitsMax=<10 0 20>, angularLimitsMin=<-10 -20 -20>, angularTarget=<1 0 0>}
      localJointOffset=<0.1 0 0>
     }]
   }
   n15 = Anim Dynamics - Bone: dyn_shoulder_zippo_r_01(ComponentPose=@n14, Alpha=1, ExternalForce="0,0,0") {
    angularDampingOverride=1
    angularSpringConstant=200
    bOverrideAngularDamping=true
    bUseGravityOverride=true
    simulationSpace=RootRelative
    alphaBoolBlend {blendOption=Linear}
    boundBone {boneName=dyn_shoulder_zippo_r_01}
    componentAppliedLinearAccClamp=<2000 2000 2000>
    componentLinearAccScale=<1 1 1>
    componentLinearVelScale=<1 1 1>
    gravityOverride {z=-980}
    physicsBodyDefinitions [{
      boundBone {boneName=dyn_shoulder_zippo_r_01}
      boxExtents=<2 2 2>
      constraintSetup {angularLimitsMax=<10 0 20>, angularLimitsMin=<-10 -25 -20>, angularTarget=<1 0 0>}
      localJointOffset=<0.1 0 0>
     }]
   }
   n16 = Anim Dynamics - Bone: dyn_necklace_01(ComponentPose=@n15, Alpha=1, ExternalForce="0,0,0") {
    angularDampingOverride=0.999
    angularSpringConstant=200
    bOverrideAngularDamping=true
    bUseGravityOverride=true
    bUsePlanarLimit=false
    simulationSpace=RootRelative
    alphaBoolBlend {blendOption=Linear}
    boundBone {boneName=dyn_necklace_01}
    componentAppliedLinearAccClamp=<2000 2000 2000>
    componentLinearAccScale=<1 1 1>
    componentLinearVelScale=<1 1 1>
    gravityOverride {z=-980}
    physicsBodyDefinitions [{
      boundBone {boneName=dyn_necklace_01}
      boxExtents=<2 2 2>
      constraintSetup {angularLimitsMax=<10 0 0>, angularLimitsMin=<-10 -40 0>, angularTarget=<1 0 0>}
      localJointOffset=<0.1 0 0>
     }]
   }
   n17 = Anim Dynamics - Bone: dyn_pistol_l_01(ComponentPose=@n16, Alpha=1, ExternalForce="0,0,0") {
    angularDampingOverride=0.999
    angularSpringConstant=200
    bOverrideAngularDamping=true
    bOverrideLinearDamping=true
    bUseGravityOverride=true
    linearDampingOverride=0.999
    simulationSpace=RootRelative
    alphaBoolBlend {blendOption=Linear}
    boundBone {boneName=dyn_pistol_l_01}
    componentAppliedLinearAccClamp=<2000 2000 2000>
    componentLinearAccScale=<1 1 1>
    componentLinearVelScale=<1 1 1>
    gravityOverride {z=-980}
    physicsBodyDefinitions [{
      boundBone {boneName=dyn_pistol_l_01}
      boxExtents=<10 10 10>
      constraintSetup {angularLimitsMax=<0 0 0>, angularLimitsMin=<0 0 0>, angularTarget=<1 0 0>}
      localJointOffset=<0.1 0 0>
     }]
   }
   Save cached pose 'toAnimDyn01'(Pose=Component To Local(ComponentPose=@n17)) {cacheName=toAnimDyn01}
  ]}
 EventGraph {nodes [
   Event BlueprintUpdateAnimation()
   > n1 = Gravity Override State Function DEPRECATED(
    CharacterPartAnimInstance=Self-Reference()
    GameTimeInSeconds=GetGameTimeInSeconds()
    GravityOverrideParams=Get Gravity Override Params()
    RBAN Params=Get RBAN Params()
    DebugJointName=dyn_hackBottomArmor
   )
   > Set Override World Gravity(Override World Gravity:Vector=@n1) -> (Output_Get:Vector)
   > n2 = PlayerMovement(AnimInstance=Self-Reference(), XDotProductCoefficient=1, YDotProductCoefficient=1, ZDotProductCoefficient=1)
   > Set Left Right Movement(Left Right Movement:double=@n2) -> (Output_Get:double)
   > Set Is Not Cape(Is Not Cape:bool=NOT Boolean(A=CheckForBackbling().bIsBackblingACape)) -> (Output_Get:bool)
  ]}
]
variables [
 Gravity Override Params {type=GravityOverrideParamsStruct, default {onGround_Moving {gravityOverride=<0 -450 -450>}}}
 Is Not Cape {type=boolean}
 Left Right Movement {type=number}
 Override World Gravity {type=Vector}
 RBAN Params {
  type=RBANParamsStruct
  default {
   dBNO {simSpaceSettings {
     maxAngularAcceleration=1000
     maxAngularVelocity=3
     maxLinearAcceleration=300
     maxLinearVelocity=700
     velocityScaleZ=0
     externalLinearDragV=<0 0.5 0>
    }}
   dBNO_Carried {simSpaceSettings {
     maxAngularAcceleration=1000
     maxAngularVelocity=3
     maxLinearAcceleration=300
     maxLinearVelocity=700
     velocityScaleZ=0
     externalLinearDragV=<0 0.5 0>
    }}
   emoteOrMelee {simSpaceSettings {
     maxAngularAcceleration=1000
     maxAngularVelocity=3
     maxLinearAcceleration=300
     maxLinearVelocity=700
     velocityScaleZ=0
     externalLinearDragV=<0 0.5 0>
    }}
   front_End {simSpaceSettings {
     maxAngularAcceleration=1000
     maxAngularVelocity=3
     maxLinearAcceleration=300
     maxLinearVelocity=700
     velocityScaleZ=0
     externalLinearDragV=<0 0.5 0>
    }}
   onGround_Crouch_Moving {simSpaceSettings {
     maxAngularAcceleration=1000
     maxAngularVelocity=3
     maxLinearAcceleration=300
     maxLinearVelocity=700
     velocityScaleZ=0
     externalLinearDragV=<0 0.5 0>
    }}
   onGround_Falling {simSpaceSettings {
     maxAngularAcceleration=1000
     maxAngularVelocity=3
     maxLinearAcceleration=300
     maxLinearVelocity=700
     velocityScaleZ=0
     externalLinearDragV=<0 0.5 0>
    }}
   onGround_Moving {simSpaceSettings {
     maxAngularAcceleration=1000
     maxAngularVelocity=3
     maxLinearAcceleration=300
     maxLinearVelocity=700
     velocityScaleZ=0
     externalLinearDragV=<0 0.5 0>
    }}
   onGround_Standing {simSpaceSettings {
     maxAngularAcceleration=1000
     maxAngularVelocity=3
     maxLinearAcceleration=300
     maxLinearVelocity=700
     velocityScaleZ=0
     externalLinearDragV=<0 0.5 0>
    }}
   rail_Grinding {simSpaceSettings {
     maxAngularAcceleration=1000
     maxAngularVelocity=3
     maxLinearAcceleration=300
     maxLinearVelocity=700
     velocityScaleZ=0
     worldAlpha=0.1
     externalLinearDragV=<0 0.5 0>
    }}
   riding_Moving {simSpaceSettings {
     maxAngularAcceleration=1000
     maxAngularVelocity=3
     maxLinearAcceleration=300
     maxLinearVelocity=700
     velocityScaleZ=0
     externalLinearDragV=<0 0.5 0>
    }}
   riding_Moving_Targeting {simSpaceSettings {
     maxAngularAcceleration=1000
     maxAngularVelocity=3
     maxLinearAcceleration=300
     maxLinearVelocity=700
     velocityScaleZ=0
     externalLinearDragV=<0 0.5 0>
    }}
   riding_Standing {simSpaceSettings {
     maxAngularAcceleration=1000
     maxAngularVelocity=3
     maxLinearAcceleration=300
     maxLinearVelocity=700
     velocityScaleZ=0
     externalLinearDragV=<0 0.5 0>
    }}
   skydive_DivingDown {simSpaceSettings {
     maxAngularAcceleration=1000
     maxAngularVelocity=3
     maxLinearAcceleration=300
     maxLinearVelocity=700
     velocityScaleZ=0
     externalLinearDragV=<0 0.5 0>
    }}
   skydive_FloatingDown {
    componentLinearAccScale=<0 0.5 0>
    componentLinearVelScale=<0 0.5 0>
    simSpaceSettings {
     maxAngularAcceleration=1000
     maxAngularVelocity=3
     maxLinearAcceleration=300
     maxLinearVelocity=700
     velocityScaleZ=0
     externalLinearDragV=<0 0.5 0>
    }
   }
   skydive_GoingUp {simSpaceSettings {
     maxAngularAcceleration=1000
     maxAngularVelocity=3
     maxLinearAcceleration=300
     maxLinearVelocity=700
     velocityScaleZ=0
     externalLinearDragV=<0 0.5 0>
    }}
   skydive_Parachuting {
    componentLinearAccScale=<0 0.1 0.3>
    componentLinearVelScale=<0 0.1 0.3>
    simSpaceSettings {
     maxAngularAcceleration=1000
     maxAngularVelocity=3
     maxLinearAcceleration=300
     maxLinearVelocity=700
     velocityScaleZ=0
     externalLinearDragV=<0 0.5 0>
    }
   }
   skydive_Surfing {simSpaceSettings {
     maxAngularAcceleration=1000
     maxAngularVelocity=3
     maxLinearAcceleration=300
     maxLinearVelocity=700
     velocityScaleZ=0
     externalLinearDragV=<0 0.5 0>
    }}
   swimming {simSpaceSettings {
     maxAngularAcceleration=1000
     maxAngularVelocity=3
     maxLinearAcceleration=300
     maxLinearVelocity=700
     velocityScaleZ=0
     externalLinearDragV=<0 0.5 0>
    }}
   tactical_Sprint {simSpaceSettings {
     maxAngularAcceleration=1000
     maxAngularVelocity=3
     maxLinearAcceleration=300
     maxLinearVelocity=700
     velocityScaleZ=0
     externalLinearDragV=<0 0.5 0>
    }}
   ziplining {simSpaceSettings {
     maxAngularAcceleration=1000
     maxAngularVelocity=3
     maxLinearAcceleration=300
     maxLinearVelocity=700
     velocityScaleZ=0
     externalLinearDragV=<0 0.5 0>
    }}
  }
 }
]
